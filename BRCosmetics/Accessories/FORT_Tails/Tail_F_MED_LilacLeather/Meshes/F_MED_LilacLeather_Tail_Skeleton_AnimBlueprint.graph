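# Animation Blueprint for the LilacLeather tail accessory on the shared Fortnite base tail skeleton.
# The AnimGraph passes the input pose through a Modify Bone node that adds a 95-degree bone-space roll to the tail's main root joint.
# A back bling check sets the "Hide Element 01" scalar on the part's dynamic material, on for capes or long back blings and off otherwise.
# The update and customization-loaded events have no logic attached.
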AnimBlueprint /BRCosmetics/Accessories/FORT_Tails/Tail_F_MED_LilacLeather/Meshes/F_MED_LilacLeather_Tail_Skeleton_AnimBlueprint
target_skeleton=/Game/Accessories/FORT_Tails/Common/Fortnite_Base_Tail_Skeleton
graphs [
 AnimGraph {nodes [
   n1 = "Transform (Modify) Bone - Bone: C_Root_Main_Root_Jnt"!(ComponentPose=Local To Component(LocalPose=Input Pose()), bAlphaBoolEnabled:bool=True, Translation="0,0,0", Rotation="0,0,95", Scale="1,1,1") -> (Pose) {
    alphaInputType=Bool
    rotationMode=BMM_Additive
    rotationSpace=BCS_BoneSpace
    alphaBoolBlend {blendOption=Linear}
    boneToModify {boneName=C_Root_Main_Root_Jnt}
    rotation {roll=95}
   }
   Output Pose(Result=Component To Local(ComponentPose=@n1))
  ]}
 EventGraph {nodes [
   Event OnCharacterCustomizationLoaded()
   n1 = CheckForBackbling()
   > n2 = Branch(Condition=OR Boolean(A=@n1.bIsBackblingACape, B=@n1.bIsBackblingLong))
   > n3 = Get Dynamic Material Instance(SkeletalMeshComponent=GetPartSkeletalMeshComponent())
   > SetScalarParameterValue(self=@n3, ParameterName=Hide Element 01, Value=1)
   SetScalarParameterValue(execute=@n2.else, self=@n3, ParameterName=Hide Element 01, Value=0)
   TryGetPawnOwner()
   Event BlueprintUpdateAnimation()
  ]}
]
variables [Custom Base Mesh {type=SkeletalMeshComponent}]
